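# Sound cue for the sonic boom heard when an Icicle Larva spawns in the SkyDrop gameplay.
# A single wave player feeds a modulator that slightly randomizes pitch downward for variation between plays.
# It uses a dedicated Icicle Larva spawn attenuation with a very large max distance, and routes through a third-person ranged weapon fire sound class.
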SoundCue /IcicleLarvaGameplay/Gameplay/SkyDrop/Audio/IcicleLarva_Spawn_SonicBoom_Cue
root=Modulator_0
nodes [
 {id=Modulator_0, children [WavePlayer_1], properties {pitchMax=1, pitchMin=0.9}}
 {id=WavePlayer_1, properties {soundWaveAssetPtr=&/IcicleLarva_Spawn_SonicBoom_01}}
]
properties {
 attenuationSettings=/IcicleLarvaGameplay/Gameplay/SkyDrop/Audio/Sounds/Attenuation/Fort_IcicleLarva_Spawn_Att
 maxDistance=2000000
 soundClassObject=&Fort_WeaponRanged_Fire_3P_PSM
 volumeMultiplier=1
}
