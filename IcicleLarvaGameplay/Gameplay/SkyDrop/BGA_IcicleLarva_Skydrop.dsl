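; Sky-drop gameplay actor for the Icicle Larva that triggers the drop and landing effects through gameplay cues.
; On BeginPlay it executes the FX cue, then waits briefly before the second cue to mimic speed-of-sound lag.
; PlayLandedFX executes a replicated landed cue at the actor's location plus a configurable offset, with an upward normal; nothing in this graph calls it.
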(event EventBeginPlay
  (|CallParentFunction)
  ; FX
  (GameplayCue|ExecuteGameplayCue self (Variables|Default|GetGCNFX))
  (Utilities|FlowControl|Delay 1.5)
  ; wait for the speed of sound :)
  (GameplayCue|ExecuteGameplayCue self (Variables|Default|GetGameplay%20Cue%20Tag)))

(fn PlayLandedFX
  (Ability|ExecuteGameplayCueWithParamsReplicated (Variables|Default|GetGC_Landed) (Ability|GameplayCue|MakeGameplayCueParameters :NormalizedMagnitude 0.0 :RawMagnitude 0.0 :Location (+ (Transformation|GetActorLocation) (Variables|Default|GetLandingGCOffset)) :Normal "0,0,1.000000" :GameplayEffectLevel 1 :AbilityLevel 1 :bReplicateLocationWhenUsingMinimalRepProxy false)))

(fn ConstructionScript
  (|CallParentFunction))
