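; Icicle Larva sky-drop container: knocks back nearby players and vehicles on landing, then becomes a map-marked pickup that transforms a player.
; Server-side capsule overlap locks it, reverts stand-in pawns, handles forced vehicle exit and grants the special item.
; The intro knockback tag, or a failsafe timer, triggers RepNotify-driven cleanup.
; Cleanup destroys tagged builds in a tunable volume, removes markers, plays a GameplayCue and destroys the actor.
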(event EventBeginPlay
  (bind _self self)
  (if (not (Variables|Default|GetbLocked))
    (CallFunction|SetDestroyVolumeParams)
    (Variables|Default|SetbSetCapsuleOff true)
    (|CallParentFunction)
    (Utilities|FlowControl|Sequence
      (:then_0
        (Networking|SwitchHasAuthority
          (:Authority
            (CallFunction|AnalyticsSpawnItem)
            ;[0.05718 0.1229 0.15 0.5] On Land Knockback Players and Vehicles
            (bind (_asynctaskproxy _targetinghandle) (Class|AsyncActionPerformTargeting|PerformTargetingAsyncAction 0 "/IcicleLarvaGameplay/Gameplay/Targeting/TP_IcicleLarva_Intro_Pushback.TP_IcicleLarva_Intro_Pushback" true)
              (:Targeted
                (for _array_element (TargetingSystem|TargetingResults|GetTargetingResultsActors (GameInstanceSubsystems|GetTargetingSubsystem) _targetinghandle)
                  (Utilities|IsValid _array_element
                    (:Is%20Valid
                      (Utilities|IsValid (Ability|GetAbilitySystemComponent _array_element)
                        (:Is%20Valid
                          (bind _asfort_ability_system_component (Utilities|Casting|CastToFortAbilitySystemComponent (Ability|GetAbilitySystemComponent _self)))
                          (Fortnite|ApplyGameplayEffectsToActor _array_element (Variables|Default|GetGEsToApply) (Transformation|GetActorLocation) false 0 (Variables|Default|GetAbilitySystem) (GameplayEffects|MakeFortEffectContext _asfort_ability_system_component _self 0 _self 0))))))))))))
      (:then_1
        (CallFunction|SetCrystalCollision (> (ScalableFloat|GetValueatLevel (Variables|Default|GetCrystalCollisonEnabled)) 0))
        (Utilities|Time|SetTimerbyEvent (delegate Custom|ActivatePickup) 0.5)))))

(event Custom|ActivatePickup
  (Variables|Default|SetbSetCapsuleOff false)
  ;[0.3843 0.151 1] UI Ping
  (HUDPings|MarkActor self) ; HUD Ping
)

(event OnComponentBeginOverlap(OverlapCapsule)
  (params (OverlappedComponent :type Primitive%20Component%20Object%20Reference) (OtherActor :type Actor%20Object%20Reference) (OtherComp :type Primitive%20Component%20Object%20Reference) (OtherBodyIndex :type Integer) (bFromSweep :type Boolean) (SweepResult :type Hit%20Result%20Structure%20%28by%20ref%29))
  (bind _playerpawn (Default|GetPlayerPawnfromActor OtherActor))
  (bind _player_pawn (Variables|Default|GetPlayer%20Pawn))
  (bind _returnvalue (Ability|GetAbilitySystemComponent _player_pawn))
  (Networking|SwitchHasAuthority
    (:Authority
      (Utilities|IsValid _playerpawn
        (:Is%20Valid
          (Variables|Default|SetPlayer%20Pawn _playerpawn)
          (if (Math|Boolean|ORBoolean (Pawn|IsBotControlled (Variables|Default|GetPlayer%20Pawn)) (Health|IsDBNO (Variables|Default|GetPlayer%20Pawn)) (Variables|Default|GetbLocked))
            _
            (elif (GameplayTags|HasMatchingGameplayTag (Variables|Default|GetPlayer%20Pawn) "(TagName=\"Granted.StandInPawn.ReplacingOriginal\")")
              ;[0.15 0.05448 0.0999 0.502] Checking if the player is transformed into BeastGirl, if so, revert to normal pawn and wait to resume
              (CallFunction|HandleReplacedPawn)
              (else
                (Variables|Default|SetbLocked true)
                (if (Vehicle|IsinVehicle (Variables|Default|GetPlayer%20Pawn))
                  (Utilities|IsValid _returnvalue
                    (:Is%20Valid
                      ; Force exit from vehicle
                      (GameplayEffects|RemoveActiveEffectswithGrantedTags _returnvalue "(GameplayTags=((TagName=\"Gameplay.Action.Player.AthenaInVehicle\")))")
                      (bind _asfort_player_controller_athena (Utilities|Casting|CastToFortPlayerControllerAthena (Pawn|GetController (Variables|Default|GetPlayer%20Pawn)))
                        (:then
                          (Inventory|GiveItemtoInventoryOwner _asfort_player_controller_athena "/IcicleLarvaGameplay/Gameplay/AGID_IcicleLarva_Special.AGID_IcicleLarva_Special")
                          (CallFunction|AnalyticsTransformPlayer _asfort_player_controller_athena)
                          (Variables|Default|SetbSetCapsuleOff true)
                          ; Wait for tag added on activation in GA_IcicleLarvaIntroKnockback
                          (Ability|Tasks|WaitGameplayTagAddToActor _asfort_player_controller_athena "(TagName=\"Ability.IcicleLarva.Intro.Knockback_Start\")" true
                            (:then
                              ; Failsafe Cleanup
                              (Utilities|Time|SetTimerbyEvent (delegate Custom|CustomEvent) 1e+01))
                            (:Added
                              (Variables|Default|SetbReadyForCleanupSequence true))))
                        (:CastFailed
                          _)))
                    (:Is%20Not%20Valid
                      _))
                  (Variables|Default|SetbLocked false)
                  (else
                    _))))))))))

(event Custom|CustomEvent
  ; Failsafe Cleanup
  (Variables|Default|SetbReadyForCleanupSequence true))

(event Custom|ResumeGranting
  (bind _player_pawn (Variables|Default|GetPlayer%20Pawn))
  (bind _returnvalue (Ability|GetAbilitySystemComponent _player_pawn))
  (if (Vehicle|IsinVehicle (Variables|Default|GetPlayer%20Pawn))
    (Utilities|IsValid _returnvalue
      (:Is%20Valid
        ; Force exit from vehicle
        (GameplayEffects|RemoveActiveEffectswithGrantedTags _returnvalue "(GameplayTags=((TagName=\"Gameplay.Action.Player.AthenaInVehicle\")))")
        (bind _asfort_player_controller_athena (Utilities|Casting|CastToFortPlayerControllerAthena (Pawn|GetController (Variables|Default|GetPlayer%20Pawn)))
          (:then
            (Inventory|GiveItemtoInventoryOwner _asfort_player_controller_athena "/IcicleLarvaGameplay/Gameplay/AGID_IcicleLarva_Special.AGID_IcicleLarva_Special")
            (CallFunction|AnalyticsTransformPlayer _asfort_player_controller_athena)
            (Variables|Default|SetbSetCapsuleOff true)
            ; Wait for tag added on activation in GA_IcicleLarvaIntroKnockback
            (Ability|Tasks|WaitGameplayTagAddToActor _asfort_player_controller_athena "(TagName=\"Ability.IcicleLarva.Intro.Knockback_Start\")" true
              (:then
                ; Failsafe Cleanup
                (Utilities|Time|SetTimerbyEvent (delegate Custom|CustomEvent) 1e+01))
              (:Added
                (Variables|Default|SetbReadyForCleanupSequence true))))
          (:CastFailed
            _)))
      (:Is%20Not%20Valid
        _))
    (Variables|Default|SetbLocked false)
    (else
      _)))

(event Custom|HandleReplacedPawn
  (bind _player_pawn (Variables|Default|GetPlayer%20Pawn))
  (bind _playerstate (Variables|/Script/Engine.Pawn|GetPlayerState _player_pawn))
  ;[0.15 0.05448 0.0999 0.502] Use the player state to wait for the transition - once it is done we can resume the regular flow
  (if (GameplayTags|HasMatchingGameplayTag _player_pawn "(TagName=\"Granted.StandInPawn.AllowsForcedTransformationBack\")")
    (Utilities|IsValid _playerstate
      (:Is%20Valid
        (Variables|Default|SetTargetPlayerState _playerstate)
        (bind _asynctaskproxy (Ability|Async|WaitGameplayEventToActor (Variables|Default|GetTargetPlayerState) "(TagName=\"Athena.StandInPawn.Event.ReturnedToOriginal\")" true)
          (:then
            (Variables|Default|SetWaitForBeastGirlAction _asynctaskproxy)
            (Ability|SendGameplayEventtoActor (Variables|Default|GetPlayer%20Pawn) "(TagName=\"Athena.StandInPawn.Event.ForceOutTransformation\")")
            (Variables|Default|SetbLocked true))
          (:EventReceived
            ;[0.15 0.05448 0.0999 0.502] Wait an extra frame for pawn swapping deferred processes
            (bind _asfort_player_pawn (Utilities|Casting|CastToFortPlayerPawn (PlayerState|GetPawn (Variables|Default|GetTargetPlayerState))))
            (Utilities|FlowControl|DelayUntilNextTick)
            (Variables|Default|SetPlayer%20Pawn _asfort_player_pawn)
            (CallFunction|ResumeGranting)
            (Ability|Async|EndAction (Variables|Default|GetWaitForBeastGirlAction))))))))

(event Custom|Analytics_SpawnItem
  ; Analytics Events
  (bind _returnvalue (IcicleLarva|GetIcicleLarvaAnalyticsComponent 0))
  (Utilities|IsValid _returnvalue
    (:Is%20Valid
      (Class|IcicleLarvaAnalyticsComponent|NotifyItemSpawn _returnvalue (Transformation|GetActorLocation)))))

(event Custom|Analytics_TransformPlayer
  (params (PlayerController :type Fort%20Player%20Controller%20Object%20Reference))
  ; Analytics Events
  (bind _returnvalue (IcicleLarva|GetIcicleLarvaAnalyticsComponent 0))
  (Utilities|IsValid _returnvalue
    (:Is%20Valid
      (Class|IcicleLarvaAnalyticsComponent|NotifyPlayerTransform _returnvalue PlayerController))))

(event Custom|Cleanup
  (Utilities|FlowControl|Sequence
    (:then_0
      (CallFunction|DestroyPlayerBuilds))
    (:then_1
      (HUDPings|UnmarkActor)
      (GameplayCue|ExecuteGameplayCue self (Variables|Default|GetGameplay%20Cue%20Tag))
      (Utilities|FlowControl|Delay 0.3)
      (Development|SetHiddeninGame (Variables|Default|GetScene) true true)
      (Utilities|FlowControl|Delay 1.0)
      (Actor|DestroyActor))))

(fn DestroyPlayerBuilds
  (bind (_returnvalue_x _returnvalue_y _returnvalue_z) (Math|Vector|BreakVector (Transformation|GetActorLocation self)))
  (if (> (ScalableFloat|GetValueatLevel (Variables|Default|GetDestroyVolumeEnabled)) 0)
    (Networking|SwitchHasAuthority
      (:Authority
        (bind _outhits (Collision|MultiSphereTraceForObjects :Start (Math|Vector|MakeVector _returnvalue_x _returnvalue_y _returnvalue_z) :End (Math|Vector|MakeVector _returnvalue_x _returnvalue_y (+ _returnvalue_z (Variables|/Script/Engine.CapsuleComponent|GetCapsuleHalfHeight (Variables|Default|GetDestroyBuilds)))) :Radius (Variables|/Script/Engine.CapsuleComponent|GetCapsuleRadius (Variables|Default|GetDestroyBuilds)) :ObjectTypes (Variables|Default|GetObject%20Types%20to%20Cast%20For) :bTraceComplex false :DrawDebugType "None" :bIgnoreSelf true :TraceColor "(R=1.000000,G=0.000000,B=0.000000,A=1.000000)" :TraceHitColor "(R=0.000000,G=1.000000,B=0.000000,A=1.000000)" :DrawTime 0.2))
        (for _array_element _outhits
          (bind _returnvalue (GameplayTags|HasAnyMatchingGameplayTags(Message) (.HitActor (Collision|BreakHitResult _array_element)) (Variables|Default|GetBuilds%20Tag%20Container)))
          (if _returnvalue
            (Utilities|IsValid (.HitActor (Collision|BreakHitResult _array_element))
              (:Is%20Valid
                (Actor|DestroyActor (.HitActor (Collision|BreakHitResult _array_element)))))))
        (return)))
    (else
      (return))))

(fn GetPlayerPawnFromActor
  (params (Actor :type Actor%20Object%20Reference))
  (bind _asfort_player_pawn (Utilities|Casting|CastToFortPlayerPawn Actor)
    (:then
      (return _asfort_player_pawn))
    (:CastFailed
      (bind _asfort_athena_vehicle (Utilities|Casting|CastToFortAthenaVehicle Actor)
        (:then
          (return (Vehicle|GetDriver _asfort_athena_vehicle)))
        (:CastFailed
          (return 0))))))

(fn MarkActor
  (params (ActorToMark :type Actor%20Object%20Reference))
  (for _array_element (Fortnite|GetAllFortPlayerControllers true true)
    (Utilities|IsValid _array_element
      (:Is%20Valid
        (Utilities|IsValid (Class|FortPlayerController|GetMarkerComponent _array_element)
          (:Is%20Valid
            (Utilities|IsValid ActorToMark
              (:Is%20Valid
                (bind _returnvalue (MarkerComponent|MarkActoronServer (Class|FortPlayerController|GetMarkerComponent _array_element) ActorToMark false true))
                (Variables|Default|SetCrystalMarkerID _returnvalue)))))))))

(fn OnRep_bLocked)

(fn OnRep_bReadyForCleanupSequence
  (if (Variables|Default|GetbReadyForCleanupSequence)
    (CallFunction|Cleanup)))

(fn OnRep_bSetCapsuleOff
  (if (Variables|Default|GetbSetCapsuleOff)
    (Collision|SetCollisionEnabled (Variables|Default|GetOverlapCapsule) "NoCollision")
    (Development|SetHiddeninGame (Variables|Default|GetStaticMesh) true true)
    (Networking|FlushNetDormancy)
    (else
      (Collision|SetCollisionEnabled (Variables|Default|GetOverlapCapsule) "QueryAndPhysics")
      (Development|SetHiddeninGame (Variables|Default|GetStaticMesh) false false)
      (Networking|FlushNetDormancy))))

(fn SetCrystalCollision
  (params (CollisionEnabled :type Boolean))
  (if CollisionEnabled
    (Collision|SetCollisionEnabled (Variables|Default|GetIcicleLarva_Crystals) "QueryAndPhysics")
    (else
      (Collision|SetCollisionEnabled (Variables|Default|GetIcicleLarva_Crystals) "NoCollision"))))

(fn SetDestroyVolumeParams
  (bind _destroybuilds (Variables|Default|GetDestroyBuilds))
  (Components|Capsule|SetCapsuleHalfHeight _destroybuilds (ScalableFloat|GetValueatLevel (Variables|Default|GetDestroyVolumeHalfHeight)) true)
  (Components|Capsule|SetCapsuleRadius _destroybuilds (ScalableFloat|GetValueatLevel (Variables|Default|GetDestroyVolumeRadius)) true)
  (return))

(fn UnmarkActor
  (bind _slatebrush (Utilities|Struct|MakeSlateBrush "(SpecifiedColor=(R=1.000000,G=1.000000,B=1.000000,A=1.000000),ColorUseRule=UseColor_Specified)" "Image" "NoTile" "NoMirror" (struct :X 0.0 :Y 0.0) "(Left=0.000000,Top=0.000000,Right=0.000000,Bottom=0.000000)" 0 "(CornerRadii=(X=0.000000,Y=0.000000,Z=0.000000,W=0.000000),Color=(SpecifiedColor=(R=0.000000,G=0.000000,B=0.000000,A=0.000000),ColorUseRule=UseColor_Specified),Width=0.000000,RoundingType=HalfHeightRadius,bUseBrushTransparency=False)"))
  (bind _athenaspecialactor (Variables|Default|GetAthenaSpecialActor))
  (for _array_element (Fortnite|GetAllFortPlayerControllers true true)
    (Utilities|IsValid _array_element
      (:Is%20Valid
        (Utilities|IsValid (Class|FortPlayerController|GetMarkerComponent _array_element)
          (:Is%20Valid
            (MarkerComponent|UnmarkActoronClient (Class|FortPlayerController|GetMarkerComponent _array_element) (Variables|Default|GetCrystalMarkerID)))))))
  (Variables|/Script/FortniteGame.AthenaSpecialActorComponent|SetSpecialActorCompassIconBrush _slatebrush _athenaspecialactor) ; Remove compass icon
  (Variables|/Script/FortniteGame.AthenaSpecialActorComponent|SetSpecialActorMinimapIconBrush _slatebrush _athenaspecialactor) ; Remove minimap icon
)

(fn ConstructionScript
  (|CallParentFunction))
